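# Niagara system that blows orange-brown leaf sprites through the air, apparently for the Season 23 front end, using one GPU emitter.
# Loop-age curves vary spawn rate, speed and spin in gusts, with leaves spawning in a growing hemisphere.
# Gravity, wind, curl noise, point attraction, drag and collision drive motion under a speed limit.
# Sub-UV animation, fades and camera-distance sizing shape the look; terrain-related modules are disabled.
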NiagaraSystem /Game/NeverCook/FX/FrontEndFX/S23/NS_Frontend_S23_Leaves
dependencies {
 usedDataInterfaces [
  &NiagaraDataInterfaceCamera
  &NiagaraDataInterfaceCurve
  &NiagaraDataInterfaceCollisionQuery
  &NiagaraDataInterfaceGBuffer
  &NiagaraDataInterfaceArrayFloat
 ]
 usedDynamicInputs []
 usedModules [
  &SystemState
  &self:SystemSpawnScript.SystemScriptSource.SystemScriptGraph.NiagaraNodeAssignment_0.SetVariables_#1
 ]
}
emitters {LeavesInAir {
  simTarget=GPUComputeSim
  emitterUpdateScript [
   /SetVariables(
    Emitter.CameraPostion:Vector3f=&ReturnCameraVector(Camera Property=::NewEnumerator0, Coordinate Space=::Simulation)
    Emitter.Emitter_HeightOffset:NiagaraFloat=$System.System_HeightOffset
    Emitter.Overall_SpeedMult=0.2
    Emitter.SpawnRate_Curve:NiagaraFloat=&FloatFromCurve(
     FloatCurve=DataInterface() {propertyValues {Curve {keys [{arriveTangent=17.5837, time=0, value=0}, <at 0.1 1>, <at 0.2 0.1>, <at 0.6 0.3>, <at 0.7 0.02>, <at 1 0.01>]}}}
     CurveIndex=$Emitter.NormalizedLoopAge
    )
    Emitter.SpawnSpeed_Curve:NiagaraFloat=&FloatFromCurve(
     FloatCurve=DataInterface() {propertyValues {Curve {keys [
         {arriveTangent=30.2685, interpMode=RCIM_Cubic, leaveTangent=30.2685, tangentMode=RCTM_User, time=0.0009, value=0.0215}
         {arriveTangent=-0.0619, interpMode=RCIM_Cubic, leaveTangent=-0.0619, tangentMode=RCTM_User, time=0.0989, value=0.3515}
         <at 0.297 0.0032>
         <at 0.3192 0.3332>
         <at 0.3525 0.0398>
         <at 0.6171 0.3148>
         <at 0.9939 0.3698>
        ]}}}
     CurveIndex=&Multiply_Float(A=$Emitter.NormalizedLoopAge, B=&RandomRangeFloat(Minimum=1, Maximum=5))
     Scale Curve=&Multiply_Float(
      A=&Lerp_Float(A=5, B=10, Alpha=&UniformRangedFloat(Minimum=0.25, Maximum=1))
      B=1
     )
    )
    Emitter.SpriteRotRate:NiagaraFloat=&FloatFromCurve(
     FloatCurve=DataInterface() {propertyValues {Curve {keys [
         <at 0.0041 0.1315>
         <at 0.1088 1.9664>
         {arriveTangent=-3.3757, interpMode=RCIM_Cubic, leaveTangent=-3.3757, tangentMode=RCTM_User, time=0.3609, value=0.6391}
         {arriveTangent=-1, interpMode=RCIM_Cubic, leaveTangent=-1, tangentMode=RCTM_User, time=0.5614, value=0.4386}
         <at 0.7086 1.6472>
         <at 1 0>
        ]}}}
     CurveIndex=$Emitter.NormalizedLoopAge
     Scale Curve=500
    )
   ) {isSetParameters=true}
   &EmitterState(
    Enable Distance Culling=true
    Life Cycle Mode=::NewEnumerator1
    Loop Duration=&RandomRangeFloat(Minimum=6, Maximum=10)
    MaxDistance=10000
    Module.bApplySpawnCountScaleByDistance=true
    Module.bUseMaxDistance=true
    Module.bUseMinDistance=false
    Recalculate Duration Each Loop=true
    Scalability Mode=::NewEnumerator1
    Scale Spawn Count=true
    Spawn Count Scale By Distance=DataInterface() {propertyValues {Curve {keys [<at 0.75 0.9657>,<at 1 0.5>]}}}
    UseLoopDelay=false
   )
   &SpawnRate(SpawnRate=&Multiply_Float(A=$Emitter.SpawnRate_Curve, B=30), Use Spawn Probability=false)
  ]
  particleSpawnScript [
   /SetVariables(
    Particles.HeightOffset=200
    Particles.UporDown:NiagaraFloat=&UniformAOrBFloat(
     A=&RandomRangeFloat(Minimum=12, Maximum=15)
     B=&RandomRangeFloat(Minimum=250, Maximum=1500)
     Distribution Weight=0.8
    )
   ) {isSetParameters=true}
   /Niagara/Modules/Spawn/Initialization/InitializeParticle(
    Color=&RandomRangeLinearColor(Minimum=<rgba 0.3372 0.2195 0.0369 1>, Maximum=<rgba 0.8958 0.4472 0.07 1>)
    Lifetime=&UniformRangedFloat(Minimum=6, Maximum=8)
    Module.Write Color=true
    Module.Write Lifetime=true
    Module.Write Mass=true
    Module.Write Material Random=true
    Module.Write Position=false
    Module.Write Scale=false
    Module.Write SpriteRotation=true
    Module.Write SpriteSize=true
    Position=&SimulationPosition()
    Sprite Rotation=&UniformRangedFloat(Minimum=-180, Maximum=180)
    Sprite Size=&Lerp_Vector2(
     A=<8 8>
     B=<15 15>
     Alpha=&FloatFromCurve(
      FloatCurve=DataInterface() {propertyValues {Curve {keys [
          {interpMode=RCIM_Cubic, tangentMode=RCTM_User, time=0, value=0}
          {arriveTangent=1.0082, interpMode=RCIM_Cubic, leaveTangent=1.0082, tangentMode=RCTM_User, time=0.7, value=0.3}
          {arriveTangent=1.0857, interpMode=RCIM_Cubic, leaveTangent=1.0857, tangentMode=RCTM_User, time=1, value=1}
         ]}}}
      CurveIndex=&RandomRangeFloat(Minimum=0, Maximum=1)
     )
    )
   )
   &InitializeParticle(
    Color=&RandomRangeLinearColor(Minimum=<rgba 0.3372 0.2195 0.0369 1>, Maximum=<rgba 0.8958 0.4472 0.07 1>)
    Lifetime=&UniformRangedFloat(Minimum=6, Maximum=8)
    Module.Write Color=true
    Module.Write Lifetime=true
    Module.Write Mass=true
    Module.Write Material Random=true
    Module.Write Position=false
    Module.Write Scale=false
    Module.Write SpriteRotation=true
    Module.Write SpriteSize=true
    Position=&SimulationPosition()
    Sprite Rotation=&UniformRangedFloat(Minimum=-180, Maximum=180)
    Sprite Size=&Vector2DFromFloat(Value=&UniformRangedFloat(Minimum=10, Maximum=15))
   ) {enabled=false, module=InitializeParticle001}
   &AddVelocity(Velocity=&Multiply_VectorByFloat(
     Vector=&Multiply_VectorByFloat(
      Vector=&Multiply_VectorByFloat(Vector=&UniformRangedVector(Minimum=<30 50 0>, Maximum=<80 150 500>), Float=$Emitter.SpawnSpeed_Curve)
      Float=$Emitter.Overall_SpeedMult
     )
     Float=1
    ))
   /Niagara/Modules/Spawn/Location/SphereLocation(
    Hemisphere Z=true
    MaskSpawnGroup=false
    Module.Override Randomness=false
    Non Uniform Scale=&RandomRangeVector(Minimum=<-1 -1 -1>, Maximum=<1 1 1>)
    Offset=&Multiply_VectorByFloat(
     Vector=<1 0 0>
     Float=&FloatFromCurve(FloatCurve=DataInterface() {propertyValues {Curve {keys [<at 0 0>,<at 2 1>]}}}, CurveIndex=$System.LoopedAge, Scale Curve=150)
    )
    Sphere Radius=&Multiply_Float(
     A=500
     B=&FloatFromCurve(FloatCurve=DataInterface() {propertyValues {Curve {keys [<at 0 0.1>,<at 0.157 0.9198>]}}}, CurveIndex=$System.NormalizedLoopAge)
    )
    Surface Only=false
   ) {module=SphereLocation001}
   /Niagara/Modules/Math/MoveToNearestDistanceFieldSurface_GPU(Kill by Distance=500, bKillByDistance=true) {enabled=false}
   /SetVariables(Particles.InitialPos:NiagaraPosition=$Particles.Position) {isSetParameters=true}
   /Game/Effects/Niagara/Modules/Spawn/FN_Spawn_GBuffer_Color() {enabled=false}
   &CurlNoiseForce(
    Noise Frequency=25
    Noise Strength=&Multiply_Float(
     A=250
     B=&Multiply_Float(A=&Multiply_Float(A=1, B=$Emitter.Overall_SpeedMult), B=1)
    )
    Pan Noise=false
   ) {module=CurlNoiseForce001}
   &ApplyInitialForces()
   /Niagara/Modules/Update/Camera/MaintainInCameraParticleScale()
   &CameraOffset(Camera Offset Amount=-200)
  ]
  particleUpdateScript [
   /SetVariables(Particles.ParticleUpdate_RotRate:NiagaraFloat=&FloatFromCurve(
     FloatCurve=DataInterface() {propertyValues {Curve {keys [
         <at 0.0041 0.1315>
         <at 0.1088 1.9664>
         {arriveTangent=-3.3757, interpMode=RCIM_Cubic, leaveTangent=-3.3757, tangentMode=RCTM_User, time=0.3609, value=0.6391}
         {arriveTangent=-1, interpMode=RCIM_Cubic, leaveTangent=-1, tangentMode=RCTM_User, time=0.5614, value=0.4386}
         <at 0.7086 1.6472>
         {arriveTangent=-1.004, interpMode=RCIM_Cubic, leaveTangent=-1.004, tangentMode=RCTM_User, time=0.8311, value=0.9545}
         {arriveTangent=-5.6531, interpMode=RCIM_Cubic, leaveTangent=-5.6531, tangentMode=RCTM_User, time=0.8987, value=1.4563}
         <at 0.9384 0.3483>
         <at 1 0>
        ]}}}
     Scale Curve=100
    )) {isSetParameters=true}
   &ParticleState()
   /Niagara/Modules/Update/SubUV/SubUVAnimation(Number Of Frames=64)
   &ScaleSpriteSize(Scale Factor=&Multiply_Vector2(
     A=<1 1>
     B=&Vector2DFromFloat(Value=&Multiply_Float(
       A=&FloatFromCurve(FloatCurve=DataInterface() {propertyValues {Curve {keys [
            {interpMode=RCIM_Cubic, tangentMode=RCTM_User, time=-0.0061, value=0.489}
            {arriveTangent=0.3677, interpMode=RCIM_Cubic, leaveTangent=0.3677, tangentMode=RCTM_User, time=0.2892, value=0.9107}
            {arriveTangent=-1.2812, interpMode=RCIM_Cubic, leaveTangent=-1.2812, tangentMode=RCTM_User, time=0.9, value=0.9275}
            <at 1 0>
           ]}}})
       B=&FloatFromCurve(
        FloatCurve=DataInterface() {propertyValues {Curve {keys [
            {interpMode=RCIM_Cubic, tangentMode=RCTM_User, time=0, value=0.0005}
            {arriveTangent=0.0074, interpMode=RCIM_Cubic, leaveTangent=0.0074, tangentMode=RCTM_User, time=0.15, value=1.0236}
            {arriveTangent=-0.0952, interpMode=RCIM_Cubic, leaveTangent=-0.0952, tangentMode=RCTM_User, time=0.85, value=1}
            <at 0.9 0>
           ]}}}
        CurveIndex=&NormalizeDistanceRange(
         Start Position:NiagaraPosition=Linked() {linkedVariable {name=Emitter.CameraPostion, type=Vector3f}}
         End Position:NiagaraPosition=$Particles.Position
         Distance:NiagaraFloat=10000
         Invert Normalized Range:NiagaraBool=false
        )
        Scale Curve=0.85
       )
      ))
    )) {module=ScaleSpriteSize001}
   &Drag(
    Drag=&FloatFromCurve(FloatCurve=DataInterface() {propertyValues {Curve {keys [<at -0.0085 0.0857>,<at 0.0429 0.6046>,<at 0.3 0.125>,<at 1 0>]}}})
    Use Linear Drag=true
    Use Rotational Drag=false
   )
   &GravityForce(Gravity=&Lerp_Vector(
     A=<0 0 -600>
     B=<0 0 -900>
     Alpha=&FloatFromCurve(
      FloatCurve=DataInterface() {propertyValues {Curve {keys [
          {interpMode=RCIM_Cubic, tangentMode=RCTM_User, time=0, value=0}
          {arriveTangent=1.0082, interpMode=RCIM_Cubic, leaveTangent=1.0082, tangentMode=RCTM_User, time=0.7, value=0.3}
          {arriveTangent=1.0857, interpMode=RCIM_Cubic, leaveTangent=1.0857, tangentMode=RCTM_User, time=1, value=1}
         ]}}}
      CurveIndex=&RandomRangeFloat(Minimum=0, Maximum=1)
     )
    ))
   &CurlNoiseForce(
    Curl Noise Cone Mask Angle=50
    Mask Curl Noise=true
    Noise Frequency=&RandomRangeFloat(Minimum=500, Maximum=800)
    Noise Strength=&Multiply_Float(A=$Emitter.SpawnSpeed_Curve, B=500)
    Pan Noise=true
    Pan Noise Field=<0 0 1>
    Random Seed=80
   )
   &PointAttractionForce(
    Attraction Radius=1500
    Attractor Position Offset:Vector3f=<1200 1000 -200>
    Kill Within Radius=false
    Module.Use Falloff=false
    UseAttractorOffset=true
   )
   &Collision(
    Age Colliding Particles=false
    Analytical Collision Plane Position 1:Vector3f=<0 0 0>
    Analytical Collision Plane Position 2:Vector3f=<0 0 -100>
    GPU Collision Type=::NewEnumerator3
    Module.Control Rest State Via Surface Normal=false
    Module.EnableMaxCollisionCount=false
    Module.Randomize Collision Normal Vector=false
    Restitution=0.4
   )
   &SpriteRotationRate(Rotation Rate=$Emitter.SpriteRotRate)
   &ScaleColor(
    Scale Alpha=&FloatFromCurve(FloatCurve=DataInterface() {propertyValues {Curve {keys [
         {interpMode=RCIM_Cubic, tangentMode=RCTM_User, time=0, value=0}
         {arriveTangent=-0.0948, interpMode=RCIM_Cubic, leaveTangent=-0.0948, tangentMode=RCTM_User, time=0.1, value=1}
         <at 0.942 0.9825>
         <at 1 0>
        ]}}})
    Scale RGB=&MakeVectorFromLinearColorRGB(Color=&HueShiftLinearColor(
      Hue=&RandomRangeFloat(Minimum=0, Maximum=0.1)
      Saturation=1
      Value=&RandomRangeFloat(Minimum=0.5, Maximum=2)
      Linear Color=$Particles.Initial.Color
     ))
    ScaleA=true
    ScaleRGB=true
   )
   /Niagara/Modules/Update/SubUV/V2/SubUVAnimation(
    Start Frame:NiagaraInt32=&RandomRangeInt(Minimum=0, Maximum=32)
    SubUV Animation Mode:ENiagara_SubUVLookupMode=::NewEnumerator1
    SubUV Loop Count:NiagaraFloat=&SelectFloatFromArray(
     Array Sampling Mode=::NewEnumerator0
     Direct Array Index=?
     Float Selection Array=DataInterface() {propertyValues {FloatData [2,4]}}
     Random Seed=0
    )
   ) {module=SubUVAnimation001}
   &DynamicMaterialParameters(
    Index 0 Param 1=0.1
    Module.Param0WriteEnabled=true
    Module.Param1WriteEnabled=true
    Module.Param2WriteEnabled=true
    Module.Param3WriteEnabled=true
   )
   &WindForce(
    Scale=1
    Wind Speed=&Lerp_Vector(
     A=<50 0 0>
     B=<100 0 0>
     Alpha=&FloatFromCurve(
      FloatCurve=DataInterface() {propertyValues {Curve {keys [
          {interpMode=RCIM_Cubic, tangentMode=RCTM_User, time=0, value=0}
          {arriveTangent=1.0082, interpMode=RCIM_Cubic, leaveTangent=1.0082, tangentMode=RCTM_User, time=0.7, value=0.3}
          {arriveTangent=1.0857, interpMode=RCIM_Cubic, leaveTangent=1.0857, tangentMode=RCTM_User, time=1, value=1}
         ]}}}
      CurveIndex=&RandomRangeFloat(Minimum=0, Maximum=1)
     )
    )
   )
   &AerodynamicDrag(Aerodynamic Drag=0.5, Use Linear Drag=true, Use Rotational Drag=true)
   &SolveForcesAndVelocity(Clamp Velocity=true, Limit Acceleration=false, Speed Limit=250)
   &AlignSpriteToMeshOrientation(SetAlignment=true, SetFacing=true)
   /NiagaraFN/Modules/Spawn/FN_BlockParticlesOnLandScape(LandscapeHeight:NiagaraFloat=&MakeFloatFromVector(Vector=$Output.FN_SpawnParticlesOnTerrainNearPlayer.LandscapePostion, Channel=::NewEnumerator2)) {enabled=false}
  ]
  rendererClasses [&NiagaraSpriteRendererProperties]
  renderers [{rendererClass=&NiagaraSpriteRendererProperties}]
 }}
summary {
 systemName=NS_Frontend_S23_Leaves
 emitters [{emitterName=LeavesInAir, simTarget=GPUComputeSim, rendererClasses [&NiagaraSpriteRendererProperties]}]
 userVariables [User.Landscape {
   type=NiagaraDataInterfaceLandscape
   description=""
   defaultValue {
    struct=NiagaraExt_VariableValue_DataInterface
    value {
     dataInterface=&self:NiagaraDataInterfaceLandscape_0
     dataInterfaceClass=&NiagaraDataInterfaceLandscape
    }
   }
  }]
}
